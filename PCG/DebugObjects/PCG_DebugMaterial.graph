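# Unlit material for PCG debug objects on instanced and Nanite meshes, visualizing either point density or color.
# Per-instance custom data feeds named reroutes for Density, Extents, Color and Alpha, though Extents and Alpha are unused.
# The DebugColor static switch picks color or sRGB-to-linear density, passes it through a vertex interpolator and divides by eye adaptation to offset auto exposure.
# A ShadowReplace branch computes an offset collapsing vertices to the object position in shadow passes.
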Material /PCG/DebugObjects/PCG_DebugMaterial
graph {
 properties {
  bCastRayTracedShadows=false
  bUsedWithInstancedStaticMeshes=true
  bUsedWithNanite=true
  pixelDepthOffsetMode=PDOM_Legacy
  shadingModel=MSM_Unlit
  usedShadingModels=Unlit
 }
 nodes [
  n1 = NamedRerouteDeclaration(None=MaterialFunctionCall(R=PerInstanceCustomData!(DefaultValue) {dataIndex=4}, G=PerInstanceCustomData() {dataIndex=5}, B=PerInstanceCustomData() {dataIndex=6}) {materialFunction=&AppendMany}.RGB) {name=Color, nodeColor=<rgba 0 0.25 1 1>}
  n2 = PerInstanceCustomData()
  n3 = MaterialFunctionCall!(sRGB=NamedRerouteDeclaration(None=@n2) {name=Density, nodeColor=<rgba 1 0 0.0469 1>}) {materialFunction=/Engine/Functions/Engine_MaterialFunctions01/ImageAdjustment/sRGBToLinear}
  Divide(A=VertexInterpolator(VS=StaticSwitchParameter(True=@n1, False=@n3.Linear) {parameterName=DebugColor}).PS, B=EyeAdaptation().EyeAdaptation)
  NamedRerouteDeclaration(None=MaterialFunctionCall(R=PerInstanceCustomData() {dataIndex=1}, G=PerInstanceCustomData() {dataIndex=2}, B=PerInstanceCustomData() {dataIndex=3}) {materialFunction=&AppendMany}.RGB) {name=Extents, nodeColor=<rgba 1 0.6094 0 1>}
  NamedRerouteDeclaration(None=PerInstanceCustomData() {dataIndex=7}) {name=Alpha, nodeColor=<rgba 1 0.9844 0 1>}
  ShadowReplace(Default=Constant(), Shadow=Subtract(A=ObjectPositionWS(), B=WorldPosition().XYZ))
  Subtract(A=Multiply(A=@n2) {constB=2})
 ]
}
