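# Unlit two-sided PCG debug material that displays a texture parameter with channel masking and an optional texel grid overlay.
# A Custom node shows a single selected channel, including alpha, as grayscale, or multiplies RGB by the mask when several are active.
# A second Custom node draws anti-aliased texel-boundary lines from UV and GridResolution, using screen derivatives to fade them when zoomed out.
# ShowGrid scales that line mask to blend GridColor over the texture.
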Material /PCG/DebugObjects/PCG_DebugMaterialTexture
graph {
 properties {shadingModel=MSM_Unlit, twoSided=true, usedShadingModels=Unlit}
 nodes [
  MaterialExpressionTextureSampleParameter2D_0 = TextureSampleParameter2D() {parameterName=DebugTexture, texture=&Black}
  MaterialExpressionVectorParameter_2 = VectorParameter() {parameterName=ChannelMask, defaultValue=<rgba 1 1 1 1>}
  n1 = Custom!(Color=@MaterialExpressionTextureSampleParameter2D_0.RGBA, Mask=@MaterialExpressionVectorParameter_2.RGBA) {
   code="float n = dot(Mask, float4(1,1,1,1));   // active channel count\r\nif (n < 1.5)                            // exactly one (or none) active -> grayscale isolate\r\n{\r\n    return (dot(Color, Mask)).xxx;       // picks R, G, B, or A as gray\r\n}\r\nreturn Color.rgb * Mask.rgb;             // multi-channel -> mask RGB, alpha ignored for color"
   showCode=true
   inputs [
    {
     inputName=Color
     input {
      expression=&self:MaterialExpressionTextureSampleParameter2D_0
      mask=1
      maskA=1
      maskB=1
      maskG=1
      maskR=1
      outputIndex=5
     }
    }
    {
     inputName=Mask
     input {
      expression=&self:MaterialExpressionVectorParameter_2
      mask=1
      maskA=1
      maskB=1
      maskG=1
      maskR=1
      outputIndex=5
     }
    }
   ]
  }
  MaterialExpressionTextureCoordinate_1 = TextureCoordinate()
  MaterialExpressionComponentMask_0 = ComponentMask(None=VectorParameter() {parameterName=GridResolution, defaultValue=<rgba 1 1 0 0>}.RGB) {g=true, r=true}
  n2 = Custom!(UV=@MaterialExpressionTextureCoordinate_1, GridRes=@MaterialExpressionComponentMask_0) {
   code="float2 p = UV * GridRes;                      // texel-space coordinate\r\nfloat2 uvDeriv = max(fwidth(p), 1e-6);        // texels per pixel\r\n\r\nfloat lineWidth = 0.02;                       // total line thickness as a fraction of a texel\r\nfloat halfWidth = lineWidth * 0.5;\r\nfloat2 distToLine = abs(frac(p - 0.5) - 0.5); // 0 at texel boundary, 0.5 at texel center\r\n\r\n// ~1px anti-aliasing band so the constant-width line doesn't shimmer at distance\r\nfloat2 g = smoothstep(halfWidth + uvDeriv, halfWidth - uvDeriv, distToLine);\r\nfloat grid = max(g.x, g.y);\r\n\r\n// Constant texel-relative lines go sub-pixel when zoomed out, so fade them out\r\nfloat texelsPerPixel = max(uvDeriv.x, uvDeriv.y);\r\nfloat fade = 1.0 - smoothstep(0.15, 0.35, texelsPerPixel);\r\n\r\nreturn grid * fade;"
   outputType=CMOT_Float1
   showCode=true
   inputs [
    {inputName=UV, input {expression=&self:MaterialExpressionTextureCoordinate_1}}
    {inputName=GridRes, input {expression=&self:MaterialExpressionComponentMask_0}}
   ]
  }
  LinearInterpolate(
   A=@n1
   B=VectorParameter() {parameterName=GridColor, defaultValue=<rgba 1 1 1 1>}.RGB
   Alpha=Multiply(A=ScalarParameter() {parameterName=ShowGrid}, B=@n2)
  )
 ]
}
