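# Debug material for PCG debug objects, named for displaying vertex colors, using Default Lit shading with support for instanced static meshes and Nanite.
# Its graph holds a VertexColor node and two scalar constants, with no output connections shown.
# A Shadow Pass Switch returns a constant normally and object position minus world position in shadow passes, an offset that would collapse shadow geometry if used as World Position Offset.
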Material /PCG/DebugObjects/PCG_DebugMaterialVertexColors
graph {
 properties {bUsedWithInstancedStaticMeshes=true, bUsedWithNanite=true, pixelDepthOffsetMode=PDOM_Legacy, usedShadingModels=Default Lit}
 nodes [
  Constant() {r=1}
  Constant() {r=0.2}
  ShadowReplace(Default=Constant(), Shadow=Subtract(A=ObjectPositionWS(), B=WorldPosition().XYZ))
  VertexColor()
 ]
}
