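# PCG_DebugWhite is a Default Lit material in the PCG DebugObjects folder, named as a white material for debug objects.
# Its graph holds a single vector parameter defaulting to opaque white, with no connection to a material output shown.
# Usage flags enable it for instanced static meshes, Nanite, skeletal meshes and mesh deformers.
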Material /PCG/DebugObjects/PCG_DebugWhite
graph {
 properties {
  bUseTranslucencyVertexFog=false
  bUsedWithInstancedStaticMeshes=true
  bUsedWithMeshDeformer=true
  bUsedWithNanite=true
  bUsedWithSkeletalMesh=true
  pixelDepthOffsetMode=PDOM_Legacy
  usedShadingModels=Default Lit
 }
 nodes [VectorParameter() {defaultValue=<rgba 1 1 1 1>}]
}
